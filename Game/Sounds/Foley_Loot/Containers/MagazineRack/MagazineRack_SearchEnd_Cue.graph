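# Sound cue played when searching a magazine rack loot container ends, combining a magazine search foley with the "no more loot" UI sting.
# A mixer layers the UI wave, slightly reduced, over the foley wave, slightly boosted.
# The foley runs through an enveloper that fades its volume out, then a modulator adds small random pitch and volume variation.
# It uses prop-search attenuation, the loot-collect sound class and a small concurrency cap favoring the nearest sounds.
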SoundCue /Game/Sounds/Foley_Loot/Containers/MagazineRack/MagazineRack_SearchEnd_Cue
root=Mixer_12
nodes [
 {
  id=Enveloper_15
  children [WavePlayer_54]
  properties {pitchCurve {editorCurveData {keys [{time=0}]}}, volumeCurve {editorCurveData {keys [<at 0 1>,<at 0.3913 0.797>,<at 0.64 0>]}}}
 }
 {id=Mixer_12, children [WavePlayer_43, Modulator_15], properties {inputVolume [0.8,1.1]}}
 {id=Modulator_15, children [Enveloper_15], properties {pitchMax=1.07, pitchMin=0.92, volumeMax=1.07, volumeMin=0.92}}
 {id=WavePlayer_43, properties {soundWaveAssetPtr=/Game/Sounds/Fort_UI_Sounds/Gameplay/Fort_UI_NoMoreLoot}}
 {id=WavePlayer_54, properties {soundWaveAssetPtr=&/Magazine_Search07}}
]
properties {
 attenuationSettings=/Game/Sounds/Attenuation/Fort_Prop_Search_Att
 bOverrideConcurrency=true
 maxDistance=1300
 soundClassObject=&Fort_Loot_Collect
 volumeMultiplier=0.8
 concurrencyOverrides {maxCount=4, resolutionRule=StopFarthestThenPreventNew}
}
