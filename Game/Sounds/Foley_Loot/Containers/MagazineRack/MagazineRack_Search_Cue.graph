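# Sound cue for searching a magazine rack loot container, using prop-search attenuation and a PreventNew concurrency override.
# A quality-level node selects between four branches of differing detail.
# The first two mix pitch- and volume-modulated random magazine rustles, a repeat delayed about a second, and a gear-search layer, with the second drawing on fewer variations.
# The last two each play the same single optimized wave through a modulator.
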SoundCue /Game/Sounds/Foley_Loot/Containers/MagazineRack/MagazineRack_Search_Cue
root=QualityLevel_0
nodes [
 {id=Delay_1, children [Modulator_5], properties {delayMax=1.2, delayMin=1}}
 {id=Delay_31, children [Modulator_23], properties {delayMax=1.2, delayMin=1}}
 {id=Mixer_1, children [Modulator_3,Delay_1,Modulator_4], properties {inputVolume [1.1,0.75,1]}}
 {id=Mixer_14, children [Modulator_23,Delay_31,Random_13], properties {inputVolume [1.1,0.75,1]}}
 {id=Modulator_0, children [WavePlayer_0]}
 {id=Modulator_1, children [WavePlayer_1]}
 {id=Modulator_23, children [Random_9], properties {pitchMax=1.07, pitchMin=0.92, volumeMax=1.07, volumeMin=0.92}}
 {id=Modulator_3, children [Random_4], properties {pitchMax=1.07, pitchMin=0.92, volumeMax=1.07, volumeMin=0.92}}
 {id=Modulator_4, children [WavePlayer_13]}
 {id=Modulator_5, children [Random_4], properties {pitchMax=1.07, pitchMin=0.92, volumeMax=1.07, volumeMin=0.92}}
 {id=QualityLevel_0, children [Mixer_14,Mixer_1,Modulator_0,Modulator_1]}
 {id=Random_13, children [WavePlayer_45, WavePlayer_46, WavePlayer_47], properties {weights [1,1,1]}}
 {id=Random_4, children [WavePlayer_9, WavePlayer_10], properties {bRandomizeWithoutReplacement=false, weights [1,1]}}
 {id=Random_9, children [WavePlayer_48, WavePlayer_49, WavePlayer_50, WavePlayer_51], properties {weights [1,1,1,1]}}
 {id=WavePlayer_0, properties {soundWaveAssetPtr=&/Magazine_Search_opt}}
 {id=WavePlayer_1, properties {soundWaveAssetPtr=&Magazine_Search_opt}}
 {id=WavePlayer_10, properties {soundWaveAssetPtr=&/Magazine_Search03}}
 {id=WavePlayer_13, properties {soundWaveAssetPtr=/Game/Sounds/Foley_Loot/Containers/Misc/Prop_Gear_Search_05}}
 {id=WavePlayer_45, properties {soundWaveAssetPtr=&Prop_Gear_Search_05}}
 {id=WavePlayer_46, properties {soundWaveAssetPtr=/Game/Sounds/Foley_Loot/Containers/Misc/Prop_Gear_Search_06}}
 {id=WavePlayer_47, properties {soundWaveAssetPtr=/Game/Sounds/Foley_Loot/Containers/Misc/Prop_Gear_Search_01}}
 {id=WavePlayer_48, properties {soundWaveAssetPtr=&/Magazine_Search11}}
 {id=WavePlayer_49, properties {soundWaveAssetPtr=&Magazine_Search03}}
 {id=WavePlayer_50, properties {soundWaveAssetPtr=&/Magazine_Search07}}
 {id=WavePlayer_51, properties {soundWaveAssetPtr=&/Magazine_Search08}}
 {id=WavePlayer_9, properties {soundWaveAssetPtr=&Magazine_Search11}}
]
properties {
 attenuationSettings=/Game/Sounds/Attenuation/Fort_Prop_Search_Att
 bOverrideConcurrency=true
 maxDistance=1300
 soundClassObject=&Fort_Loot_Collect
 volumeMultiplier=1
 concurrencyOverrides {resolutionRule=PreventNew}
}
